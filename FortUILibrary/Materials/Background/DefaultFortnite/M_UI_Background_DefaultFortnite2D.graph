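# UI-domain translucent material for the default Fortnite 2D background.
# Its graph holds a single call to the MF_UI_Background_DefaultFortnite material function, so the background logic lives in that function rather than here.
# Automatic usage flagging in the editor and static mesh usage are both turned off.
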Material /FortUILibrary/Materials/Background/DefaultFortnite/M_UI_Background_DefaultFortnite2D
graph {
 properties {
  bAutomaticallySetUsageInEditor=false
  bUsedWithStaticMesh=false
  blendMode=BLEND_Translucent
  materialDomain=MD_UI
  usedShadingModels=Default Lit
 }
 nodes [MaterialFunctionCall() {materialFunction=&/MF_UI_Background_DefaultFortnite}]
}
