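# Translucent UI material for the SystemFortnite background, with an optional overlay fade and a top gradient mask.
# The background comes from the MF_UI_Background_SystemFortnite function.
# When Display as Overlay is on, a rotatable vertical UV gradient is remapped by Fading Range and sine-eased, otherwise it is constant.
# That result is multiplied by a saturated vertical mask offset through Top Gradient Control.
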Material /FortUILibrary/Materials/Background/SystemFortnite/M_UI_Background_SystemFortnite
graph {
 properties {
  bAutomaticallySetUsageInEditor=false
  bUsedWithStaticMesh=false
  blendMode=BLEND_Translucent
  materialDomain=MD_UI
  usedShadingModels=Default Lit
 }
 nodes [
  MaterialFunctionCall() {materialFunction=&/MF_UI_Background_SystemFortnite}
  n1 = MaterialFunctionCall() {materialFunction=&GetUserInterfaceUV}
  n2 = ScalarParameter() {group=Overlay Settings, parameterName=Overlay Rotation, sliderMax=1}
  n3 = MaterialFunctionCall(
   Input=ComponentMask(None=Rotator(Coordinate=@n1.Normalized UV, Time=@n2) {speed=6.2832}) {g=true}
   Input Low=Constant()
   Input High=ScalarParameter() {defaultValue=1, group=Overlay Settings, parameterName=Fading Range, sliderMax=1}
  ) {materialFunction=&RemapValue_0-1}
  n4 = ComponentMask(None=@n1.Normalized UV) {g=true}
  n5 = Saturate(None=ScalarParameter() {group=Background Settings, parameterName=Top Gradient Control, sliderMax=1})
  n6 = OneMinus(None=@n5)
  n7 = Subtract(A=Saturate(None=Add(A=@n4, B=@n6)), B=@n5)
  Multiply(
   A=StaticSwitchParameter(
    True=MaterialFunctionCall("Value (0-1)"=@n3.Result) {materialFunction=&MF_UI_EaseCurves}.EaseOutSine
    False=Constant() {r=1}
   ) {group=Overlay Settings, parameterName=Display as Overlay, sortPriority=0}
   B=NamedRerouteDeclaration(None=Saturate(None=@n7)) {name=Top Gradient Mask, nodeColor=<rgba 0 0 0 1>}
  )
 ]
}
