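# Debug test widget showing an example primary-content screen built on WBP_UIKit_ScreenLayout, with top nav padding turned off.
# The layout's named slots hold labeled placeholders for background, primary content, content overlay and screen overlay, with material images behind the first two.
# Another slot holds a row of three non-focusable quiet buttons bound to left, top and right face-button input actions.
# The event graph is empty, so it is purely a visual layout.
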WidgetBlueprint /Game/NeverCook/UI/Debug/TestWidgets/Layout/WBP_Test_PrimaryLayoutExample
graphs [EventGraph {links [], nodes []}]
variables []
widgets {
 info {
  inheritedWidgetCount=0
  namedSlotCount=5
  parentClass=&UserWidget
  rootWidgetClass=/FortUILibrary/UIKit/Layout/WBP_UIKit_ScreenLayout.WBP_UIKit_ScreenLayout_C
  widgetCount=15
 }
 widgets [
  WBP_UIKit_ScreenLayout {
   widgetClassPath=/FortUILibrary/UIKit/Layout/WBP_UIKit_ScreenLayout
   bIsVariable=true
   widget=WBP_UIKit_ScreenLayout
   properties {bUseTopNavPadding=false, secondaryContent_Fill=0}
  }
  Overlay_1 {widgetClassPath=&Overlay, namedSlotHost=WBP_UIKit_ScreenLayout, widget=Overlay_1}
  Image_Background {
   widgetClassPath=&Image
   bIsVariable=true
   parent=Overlay_1
   widget=Image_Background
   properties {brush {
     bIsDynamicallyLoaded=false
     drawAs=Image
     imageType=FullColor
     mirroring=NoMirror
     resourceName=None
     resourceObject=/FortUILibrary/Materials/Background/SystemFortnite/M_UI_Background_SystemFortnite
     tiling=NoTile
     imageSize=<32 32>
     margin=<ltrb 0 0 0 0>
     outlineSettings {
      bUseBrushTransparency=false
      roundingType=HalfHeightRadius
      width=0
      color {colorUseRule=UseColor_Specified, specifiedColor=<rgba 0 0 0 0>}
      cornerRadii=<0 0 0 0>
     }
     tintColor {colorUseRule=UseColor_Specified, specifiedColor=<rgba 1 1 1 1>}
     uVRegion {bIsValid=false, max=<0 0>, min=<0 0>}
    }}
   slot {class=OverlaySlot, properties {horizontalAlignment=HAlign_Fill, verticalAlignment=VAlign_Fill}}
  }
  CommonTextBlock_66 {
   widgetClassPath=&CommonTextBlock
   parent=Overlay_1
   widget=CommonTextBlock_66
   properties {
    text=Background
    font {
     bForceMonospaced=false
     bMaterialIsStencil=false
     fontMaterial=null
     fontObject=&NotoSans
     letterSpacing=0
     monospacedWidth=1
     size=21
     skewAmount=0
     typefaceFontName=NotoSans-Regular
     outlineSettings {
      bApplyOutlineToDropShadows=false
      bMiteredCorners=false
      bSeparateFillAlpha=false
      outlineMaterial=null
      outlineSize=0
      outlineColor=<rgba 0 0 0 1>
     }
    }
    shadowOffset=<0 0>
   }
   slot {class=OverlaySlot, properties {horizontalAlignment=HAlign_Center, padding=<ltrb 20 20 20 20>}}
  }
  HorizontalBox_0 {widgetClassPath=&HorizontalBox, namedSlotHost=WBP_UIKit_ScreenLayout, widget=HorizontalBox_0, properties {bOverride_Cursor=true}}
  WBP_UIKit_Button_Quiet {
   widgetClassPath=/FortUILibrary/UIKit/Buttons/WBP_UIKit_Button_Quiet
   bIsVariable=true
   parent=HorizontalBox_0
   widget=WBP_UIKit_Button_Quiet
   properties {bIsFocusable=false, triggeringInputAction {dataTable=&InputActionDataTable, rowName=GenericFaceButtonLeft}}
   slot {class=HorizontalBoxSlot, properties {padding=<ltrb 10 10 10 10>}}
  }
  WBP_UIKit_Button_Quiet_1 {
   widgetClassPath=&WBP_UIKit_Button_Quiet
   bIsVariable=true
   parent=HorizontalBox_0
   widget=WBP_UIKit_Button_Quiet_1
   properties {bIsFocusable=false, triggeringInputAction {dataTable=&InputActionDataTable, rowName=GenericFaceButtonTop}}
   slot {class=HorizontalBoxSlot, properties {padding=<ltrb 10 10 10 10>}}
  }
  WBP_UIKit_Button_Quiet_2 {
   widgetClassPath=&WBP_UIKit_Button_Quiet
   bIsVariable=true
   parent=HorizontalBox_0
   widget=WBP_UIKit_Button_Quiet_2
   properties {bIsFocusable=false, triggeringInputAction {dataTable=&InputActionDataTable, rowName=GenericFaceButtonRight}}
   slot {class=HorizontalBoxSlot, properties {padding=<ltrb 10 10 10 10>}}
  }
  Overlay {widgetClassPath=&Overlay, namedSlotHost=WBP_UIKit_ScreenLayout, widget=Overlay}
  CommonTextBlock {
   widgetClassPath=&CommonTextBlock
   parent=Overlay
   widget=CommonTextBlock
   properties {
    text=Content Overlay
    font {
     bForceMonospaced=false
     bMaterialIsStencil=false
     fontMaterial=null
     fontObject=&NotoSans
     letterSpacing=0
     monospacedWidth=1
     size=21
     skewAmount=0
     typefaceFontName=NotoSans-Regular
     outlineSettings {
      bApplyOutlineToDropShadows=false
      bMiteredCorners=false
      bSeparateFillAlpha=false
      outlineMaterial=null
      outlineSize=0
      outlineColor=<rgba 0 0 0 1>
     }
    }
    shadowOffset=<0 0>
   }
   slot {class=OverlaySlot, properties {horizontalAlignment=HAlign_Center, padding=<ltrb 20 20 20 20>}}
  }
  Overlay_2 {widgetClassPath=&Overlay, namedSlotHost=WBP_UIKit_ScreenLayout, widget=Overlay_2}
  Image_PrimaryContent {
   widgetClassPath=&Image
   bIsVariable=true
   parent=Overlay_2
   widget=Image_PrimaryContent
   properties {brush {
     bIsDynamicallyLoaded=false
     drawAs=Image
     imageType=FullColor
     mirroring=NoMirror
     resourceName=None
     resourceObject=/FortUILibrary/Materials/Background/DefaultFortnite/M_UI_Background_DefaultFortnite2D
     tiling=NoTile
     imageSize=<32 32>
     margin=<ltrb 0 0 0 0>
     outlineSettings {
      bUseBrushTransparency=false
      roundingType=HalfHeightRadius
      width=0
      color {colorUseRule=UseColor_Specified, specifiedColor=<rgba 0 0 0 0>}
      cornerRadii=<0 0 0 0>
     }
     tintColor {colorUseRule=UseColor_Specified, specifiedColor=<rgba 1 1 1 1>}
     uVRegion {bIsValid=false, max=<0 0>, min=<0 0>}
    }}
   slot {class=OverlaySlot, properties {horizontalAlignment=HAlign_Fill, verticalAlignment=VAlign_Fill}}
  }
  CommonTextBlock_1 {
   widgetClassPath=&CommonTextBlock
   parent=Overlay_2
   widget=CommonTextBlock_1
   properties {
    text=Primary Content
    font {
     bForceMonospaced=false
     bMaterialIsStencil=false
     fontMaterial=null
     fontObject=&NotoSans
     letterSpacing=0
     monospacedWidth=1
     size=21
     skewAmount=0
     typefaceFontName=NotoSans-Regular
     outlineSettings {
      bApplyOutlineToDropShadows=false
      bMiteredCorners=false
      bSeparateFillAlpha=false
      outlineMaterial=null
      outlineSize=0
      outlineColor=<rgba 0 0 0 1>
     }
    }
    shadowOffset=<0 0>
   }
   slot {class=OverlaySlot, properties {horizontalAlignment=HAlign_Center, verticalAlignment=VAlign_Center, padding=<ltrb 10 10 10 10>}}
  }
  Overlay_4 {widgetClassPath=&Overlay, namedSlotHost=WBP_UIKit_ScreenLayout, widget=Overlay_4}
  CommonTextBlock_3 {
   widgetClassPath=&CommonTextBlock
   parent=Overlay_4
   widget=CommonTextBlock_3
   properties {
    text=Screen Overlay
    font {
     bForceMonospaced=false
     bMaterialIsStencil=false
     fontMaterial=null
     fontObject=&NotoSans
     letterSpacing=0
     monospacedWidth=1
     size=21
     skewAmount=0
     typefaceFontName=NotoSans-Regular
     outlineSettings {
      bApplyOutlineToDropShadows=false
      bMiteredCorners=false
      bSeparateFillAlpha=false
      outlineMaterial=null
      outlineSize=0
      outlineColor=<rgba 0 0 0 1>
     }
    }
    shadowOffset=<0 0>
   }
   slot {class=OverlaySlot, properties {padding=<ltrb 20 20 20 20>}}
  }
 ]
}
