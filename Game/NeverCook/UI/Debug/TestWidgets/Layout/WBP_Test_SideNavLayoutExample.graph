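# Debug test widget that demonstrates the UIKit screen layout with a collapsible side navigation.
# WBP_UIKit_ScreenLayout is the root, with its named slots holding labeled placeholders for background, primary, secondary and overlay regions, plus quiet buttons bound to gamepad face-button actions.
# On Construct, a loop adds placeholder tabs to the side navigation by index, with no view model or tab ID supplied.
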WidgetBlueprint /Game/NeverCook/UI/Debug/TestWidgets/Layout/WBP_Test_SideNavLayoutExample
graphs [EventGraph {nodes [
   Event Construct()
   > n1 = For Loop(LastIndex=5)
   > AddSideNavigationTab!(
    execute
    self:WBP UIKit Side Navigation Base Object Reference=Get WBP_UIKit_SideNavigation() -> (WBP_UIKit_SideNavigation:WBP UIKit Side Navigation Object Reference)
    TabViewModel:MVVM Base Viewmodel Object Reference
    TabID:Name=None
    Index:int=@n1
   ) -> (then)
  ]}]
variables []
widgets {
 info {
  inheritedWidgetCount=0
  namedSlotCount=7
  parentClass=&UserWidget
  rootWidgetClass=/FortUILibrary/UIKit/Layout/WBP_UIKit_ScreenLayout.WBP_UIKit_ScreenLayout_C
  widgetCount=19
 }
 widgets [
  WBP_UIKit_ScreenLayout {widgetClassPath=/FortUILibrary/UIKit/Layout/WBP_UIKit_ScreenLayout, bIsVariable=true, widget=WBP_UIKit_ScreenLayout}
  Overlay_1 {widgetClassPath=&Overlay, namedSlotHost=WBP_UIKit_ScreenLayout, widget=Overlay_1}
  Image_Background {
   widgetClassPath=&Image
   bIsVariable=true
   parent=Overlay_1
   widget=Image_Background
   properties {brush {
     bIsDynamicallyLoaded=false
     drawAs=Image
     imageType=FullColor
     mirroring=NoMirror
     resourceName=None
     resourceObject=/FortUILibrary/Materials/Background/SystemFortnite/M_UI_Background_SystemFortnite
     tiling=NoTile
     imageSize=<32 32>
     margin=<ltrb 0 0 0 0>
     outlineSettings {
      bUseBrushTransparency=false
      roundingType=HalfHeightRadius
      width=0
      color {colorUseRule=UseColor_Specified, specifiedColor=<rgba 0 0 0 0>}
      cornerRadii=<0 0 0 0>
     }
     tintColor {colorUseRule=UseColor_Specified, specifiedColor=<rgba 1 1 1 1>}
     uVRegion {bIsValid=false, max=<0 0>, min=<0 0>}
    }}
   slot {class=OverlaySlot, properties {horizontalAlignment=HAlign_Fill, verticalAlignment=VAlign_Fill}}
  }
  CommonTextBlock_66 {
   widgetClassPath=&CommonTextBlock
   parent=Overlay_1
   widget=CommonTextBlock_66
   properties {
    text=Background
    font {
     bForceMonospaced=false
     bMaterialIsStencil=false
     fontMaterial=null
     fontObject=&NotoSans
     letterSpacing=0
     monospacedWidth=1
     size=21
     skewAmount=0
     typefaceFontName=NotoSans-Regular
     outlineSettings {
      bApplyOutlineToDropShadows=false
      bMiteredCorners=false
      bSeparateFillAlpha=false
      outlineMaterial=null
      outlineSize=0
      outlineColor=<rgba 0 0 0 1>
     }
    }
    shadowOffset=<0 0>
   }
   slot {class=OverlaySlot, properties {horizontalAlignment=HAlign_Center, padding=<ltrb 20 20 20 20>}}
  }
  HorizontalBox_0 {widgetClassPath=&HorizontalBox, namedSlotHost=WBP_UIKit_ScreenLayout, widget=HorizontalBox_0, properties {bOverride_Cursor=true}}
  WBP_UIKit_Button_Quiet {
   widgetClassPath=/FortUILibrary/UIKit/Buttons/WBP_UIKit_Button_Quiet
   bIsVariable=true
   parent=HorizontalBox_0
   widget=WBP_UIKit_Button_Quiet
   properties {bIsFocusable=false, triggeringInputAction {dataTable=&InputActionDataTable, rowName=GenericFaceButtonLeft}}
   slot {class=HorizontalBoxSlot, properties {padding=<ltrb 10 10 10 10>}}
  }
  WBP_UIKit_Button_Quiet_1 {
   widgetClassPath=&WBP_UIKit_Button_Quiet
   bIsVariable=true
   parent=HorizontalBox_0
   widget=WBP_UIKit_Button_Quiet_1
   properties {bIsFocusable=false, triggeringInputAction {dataTable=&InputActionDataTable, rowName=GenericFaceButtonTop}}
   slot {class=HorizontalBoxSlot, properties {padding=<ltrb 10 10 10 10>}}
  }
  WBP_UIKit_Button_Quiet_2 {
   widgetClassPath=&WBP_UIKit_Button_Quiet
   bIsVariable=true
   parent=HorizontalBox_0
   widget=WBP_UIKit_Button_Quiet_2
   properties {bIsFocusable=false, triggeringInputAction {dataTable=&InputActionDataTable, rowName=GenericFaceButtonRight}}
   slot {class=HorizontalBoxSlot, properties {padding=<ltrb 10 10 10 10>}}
  }
  Overlay {widgetClassPath=&Overlay, namedSlotHost=WBP_UIKit_ScreenLayout, widget=Overlay}
  CommonTextBlock {
   widgetClassPath=&CommonTextBlock
   parent=Overlay
   widget=CommonTextBlock
   properties {
    text=Content Overlay
    font {
     bForceMonospaced=false
     bMaterialIsStencil=false
     fontMaterial=null
     fontObject=&NotoSans
     letterSpacing=0
     monospacedWidth=1
     size=21
     skewAmount=0
     typefaceFontName=NotoSans-Regular
     outlineSettings {
      bApplyOutlineToDropShadows=false
      bMiteredCorners=false
      bSeparateFillAlpha=false
      outlineMaterial=null
      outlineSize=0
      outlineColor=<rgba 0 0 0 1>
     }
    }
    shadowOffset=<0 0>
   }
   slot {class=OverlaySlot, properties {horizontalAlignment=HAlign_Center, padding=<ltrb 20 20 20 20>}}
  }
  Overlay_2 {widgetClassPath=&Overlay, namedSlotHost=WBP_UIKit_ScreenLayout, widget=Overlay_2}
  Image_PrimaryContent {
   widgetClassPath=&Image
   bIsVariable=true
   parent=Overlay_2
   widget=Image_PrimaryContent
   properties {brush {
     bIsDynamicallyLoaded=false
     drawAs=Image
     imageType=FullColor
     mirroring=NoMirror
     resourceName=None
     resourceObject=/FortUILibrary/Materials/Background/DefaultFortnite/M_UI_Background_DefaultFortnite2D
     tiling=NoTile
     imageSize=<32 32>
     margin=<ltrb 0 0 0 0>
     outlineSettings {
      bUseBrushTransparency=false
      roundingType=HalfHeightRadius
      width=0
      color {colorUseRule=UseColor_Specified, specifiedColor=<rgba 0 0 0 0>}
      cornerRadii=<0 0 0 0>
     }
     tintColor {colorUseRule=UseColor_Specified, specifiedColor=<rgba 1 1 1 1>}
     uVRegion {bIsValid=false, max=<0 0>, min=<0 0>}
    }}
   slot {class=OverlaySlot, properties {horizontalAlignment=HAlign_Fill, verticalAlignment=VAlign_Fill}}
  }
  CommonTextBlock_1 {
   widgetClassPath=&CommonTextBlock
   parent=Overlay_2
   widget=CommonTextBlock_1
   properties {
    text=Primary Content
    font {
     bForceMonospaced=false
     bMaterialIsStencil=false
     fontMaterial=null
     fontObject=&NotoSans
     letterSpacing=0
     monospacedWidth=1
     size=21
     skewAmount=0
     typefaceFontName=NotoSans-Regular
     outlineSettings {
      bApplyOutlineToDropShadows=false
      bMiteredCorners=false
      bSeparateFillAlpha=false
      outlineMaterial=null
      outlineSize=0
      outlineColor=<rgba 0 0 0 1>
     }
    }
    shadowOffset=<0 0>
   }
   slot {class=OverlaySlot, properties {horizontalAlignment=HAlign_Center, verticalAlignment=VAlign_Center, padding=<ltrb 10 10 10 10>}}
  }
  Overlay_4 {widgetClassPath=&Overlay, namedSlotHost=WBP_UIKit_ScreenLayout, widget=Overlay_4}
  CommonTextBlock_3 {
   widgetClassPath=&CommonTextBlock
   parent=Overlay_4
   widget=CommonTextBlock_3
   properties {
    text=Screen Overlay
    font {
     bForceMonospaced=false
     bMaterialIsStencil=false
     fontMaterial=null
     fontObject=&NotoSans
     letterSpacing=0
     monospacedWidth=1
     size=21
     skewAmount=0
     typefaceFontName=NotoSans-Regular
     outlineSettings {
      bApplyOutlineToDropShadows=false
      bMiteredCorners=false
      bSeparateFillAlpha=false
      outlineMaterial=null
      outlineSize=0
      outlineColor=<rgba 0 0 0 1>
     }
    }
    shadowOffset=<0 0>
   }
   slot {class=OverlaySlot, properties {padding=<ltrb 20 20 20 20>}}
  }
  Overlay_3 {widgetClassPath=&Overlay, namedSlotHost=WBP_UIKit_ScreenLayout, widget=Overlay_3}
  Image_SecondaryContent {
   widgetClassPath=&Image
   bIsVariable=true
   parent=Overlay_3
   widget=Image_SecondaryContent
   properties {brush {
     bIsDynamicallyLoaded=false
     drawAs=Image
     imageType=FullColor
     mirroring=NoMirror
     resourceName=None
     resourceObject=&M_UI_Background_DefaultFortnite2D
     tiling=NoTile
     imageSize=<1024 512>
     margin=<ltrb 0 0 0 0>
     outlineSettings {
      bUseBrushTransparency=false
      roundingType=HalfHeightRadius
      width=0
      color {colorUseRule=UseColor_Specified, specifiedColor=<rgba 0 0 0 0>}
      cornerRadii=<0 0 0 0>
     }
     tintColor {colorUseRule=UseColor_Specified, specifiedColor=<rgba 1 0.2369 0.3598 1>}
     uVRegion {bIsValid=false, max=<0 0>, min=<0 0>}
    }}
   slot {class=OverlaySlot, properties {horizontalAlignment=HAlign_Fill, verticalAlignment=VAlign_Fill}}
  }
  CommonTextBlock_2 {
   widgetClassPath=&CommonTextBlock
   parent=Overlay_3
   widget=CommonTextBlock_2
   properties {
    text=Secondary Content
    font {
     bForceMonospaced=false
     bMaterialIsStencil=false
     fontMaterial=null
     fontObject=&NotoSans
     letterSpacing=0
     monospacedWidth=1
     size=21
     skewAmount=0
     typefaceFontName=NotoSans-Regular
     outlineSettings {
      bApplyOutlineToDropShadows=false
      bMiteredCorners=false
      bSeparateFillAlpha=false
      outlineMaterial=null
      outlineSize=0
      outlineColor=<rgba 0 0 0 1>
     }
    }
    shadowOffset=<0 0>
   }
   slot {class=OverlaySlot, properties {horizontalAlignment=HAlign_Center, verticalAlignment=VAlign_Center, padding=<ltrb 10 10 10 10>}}
  }
  WBP_UIKit_SideNavigation {
   widgetClassPath=/FortUILibrary/UIKit/Tabs/WBP_UIKit_SideNavigation
   bIsVariable=true
   namedSlotHost=WBP_UIKit_ScreenLayout
   widget=WBP_UIKit_SideNavigation
   properties {enableCollapsedState=true}
  }
 ]
}
